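# Additive unlit material for the glow of the balloon burner flame on the Sticker Battle Bus.
# A gradient texture, indexed by bounding-box height bent by axis-aligned fresnel, blends four color parameters, with panning noise available as distortion.
# Fresnel and vertical UV masks shape opacity and inner detail, while layered sine waves, normal inflation and the shared PartyBus animation function drive vertex offset.
# The optional hue shift and depth fade share one static switch name, DepthFade.
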Material /Game/Athena/Aircraft/BattleBus_Sticker/FX/M_BalloonBurner_Sticker_Flame_Glow
graph {
 properties {blendMode=BLEND_Additive, pixelDepthOffsetMode=PDOM_Legacy, shadingModel=MSM_Unlit, usedShadingModels=Unlit}
 nodes [
  n1 = TextureCoordinate()
  n2 = ComponentMask(None=@n1) {g=true}
  Add(A=@n2) {constB=0.5}
  ComponentMask(None=TextureCoordinate() {vTiling=6}) {g=true}
  Constant3Vector() {constant=<rgba 5 5 10 0>}
  n3 = MaterialFunctionCall() {materialFunction=&BoundingBoxBased_0-1_UVW}
  n4 = MaterialFunctionCall() {materialFunction=&AxisAlignedFresnel}
  n5 = OneMinus(None=@n4.Fresnel)
  n6 = Add(
   A=Add(A=@n3.B, B=ScalarParameter() {defaultValue=0.496, parameterName=Flame Core Shift})
   B=Multiply(A=@n5, B=ScalarParameter() {defaultValue=2, parameterName=Flame Core Bend})
  )
  n7 = Multiply(A=NamedRerouteUsage()) {constB=0.15}
  LinearInterpolate(
   A=ScalarParameter() {defaultValue=9, parameterName=Flame Intensity Core}
   B=ScalarParameter() {defaultValue=1, parameterName=Flame Intensity Edge}
   Alpha=Saturate(None=Add(A=@n6, B=@n7))
  ) {constA=1}
  n8 = DotProduct(
   A=MaterialFunctionCall() {materialFunction=&CameraDirectionVector}.Result
   B=Constant3Vector() {constant=<rgba 0 0 -1 0>}
  )
  n9 = OneMinus(None=Multiply(A=Add(A=@n8)) {constB=0.5})
  Multiply(A=@n9)
  n10 = VectorParameter() {parameterName=Color A, defaultValue=<rgba 1 0.0885 0.3235 0.2>}
  n11 = VectorParameter() {parameterName=Color B, defaultValue=<rgba 1.5 1.1464 0.0858 0.302>}
  n12 = VectorParameter() {parameterName=Color C, defaultValue=<rgba 0.012 1 0.2168 0.2>}
  n13 = VectorParameter() {parameterName=Color D, defaultValue=<rgba 0.011 0.4508 1 0.1>}
  n14 = Multiply(A=@n5, B=ScalarParameter() {defaultValue=0.33, parameterName=Fresnel Color Bend})
  n15 = OneMinus(None=Saturate(None=Add(A=@n3.B, B=@n14)))
  n16 = NamedRerouteUsage()
  n17 = TextureSample(UVs=Add(A=@n15, B=Multiply(A=@n16) {constB=-0.1})) {
   samplerSource=SSM_Clamp_WorldGroupSettings
   samplerType=SAMPLERTYPE_LinearGrayscale
   texture=&/T_BalloonFlame_Sticker_Gradient
  }
  NamedRerouteDeclaration(None=Multiply(A=MaterialFunctionCall(1=@n10.RGB, 2=@n11.RGB, 3=@n12.RGB, 4=@n13.RGB, A=@n17.R) {materialFunction=&Lerp_Multiple_Float3}.Lerp 4 Inputs) {desc=To emissive color}) {name=Color, nodeColor=<rgba 0 0.6016 1 1>}
  n18 = Power(Base=Multiply(A=@n4.Fresnel)) {constExponent=5}
  n19 = Multiply(A=@n18)
  n20 = Constant() {r=0.3}
  n21 = Constant() {r=0.5}
  n22 = ComponentMask(None=TextureCoordinate()) {g=true}
  n23 = Saturate(None=SmoothStep(Min=@n20, Max=Add(A=@n20, B=@n21), Value=OneMinus(None=@n22)))
  NamedRerouteDeclaration(None=Multiply(A=Multiply(A=@n19, B=@n23), B=ScalarParameter() {defaultValue=0.3, parameterName=Opacity}) {desc=To opacity}) {name=Opacity, nodeColor=<rgba 0 0.7187 1 1>}
  n24 = TextureCoordinate() {uTiling=2, vTiling=0.6}
  NamedRerouteDeclaration(None=TextureSample(UVs=Panner(Coordinate=@n24) {bFractionalPart=true, speedX=0.17, speedY=2.5}) {texture=&T_NoisePack_01}.R) {name=Distortion, nodeColor=<rgba 1 0.9375 0 1>}
  n25 = ScalarParameter() {defaultValue=1.5, parameterName=FresnelExp}
  n26 = ComponentMask(None=TextureCoordinate()) {g=true}
  n27 = Add(A=CameraVectorWS(), B=Constant3Vector() {constant=<rgba 0 0 -0.5 0>})
  n28 = Multiply(A=MaterialFunctionCall(Exponent=Multiply(A=@n25, B=@n26), Camera Vector=Normalize(VectorInput=@n27)) {materialFunction=&MF_Fresnel_FX}.Fresnel Inverted)
  n29 = Clamp(None=Add(A=OneMinus(None=@n26)) {constB=-0.75})
  n30 = Multiply(A=Multiply(A=@n29, B=@n29)) {constB=4}
  n31 = NamedRerouteUsage()
  n32 = Clamp(None=Add(A=Add(A=@n28, B=@n30) {constB=0}, B=Multiply(A=@n31) {constB=0.25}))
  n33 = OneMinus(None=@n32)
  n34 = Saturate(None=Add(A=Power(Base=@n32) {constExponent=8}, B=Power(Base=@n33) {constExponent=4}) {constB=0})
  n35 = OneMinus(None=Power(Base=OneMinus(None=@n34)) {constExponent=3})
  n36 = ScalarParameter() {defaultValue=0.3, parameterName=Inner Detail}
  NamedRerouteDeclaration(None=Saturate(None=Add(A=@n35, B=@n36))) {name=Inner Detail, nodeColor=<rgba 0 1 0.625 1>}
  n37 = Multiply(A=Frac(None=Time())) {constB=7.1}
  n38 = OneMinus(None=ComponentMask(None=TextureCoordinate()) {g=true})
  n39 = Multiply(A=Multiply(A=@n38, B=@n38)) {constB=-2}
  n40 = Multiply(A=Sine(None=Add(A=@n37, B=@n39))) {constB=1.2, desc=Amplitude}
  n41 = Add(A=Time(), B=ComponentMask(None=TextureCoordinate()) {g=true})
  n42 = TextureCoordinate()
  n43 = Add(A=Multiply(A=Time()) {constB=1.31}, B=Multiply(A=ComponentMask(None=@n42) {g=true}) {constB=0.63})
  n44 = Add(
   A=Multiply(A=Sine(None=@n41) {period=1.31}) {constB=4, desc=Amplitude}
   B=Multiply(A=Sine(None=@n43) {period=1.31}) {constB=4, desc=Amplitude}
  )
  n45 = Multiply(A=Multiply(A=VertexNormalWS()), B=Max(A=Add(A=@n40, B=@n44)) {constB=-2.5})
  n46 = Time()
  n47 = AppendVector(A=ComponentMask(None=WorldPosition().XYZ) {b=true}, B=Constant2Vector())
  n48 = Add(A=Multiply(A=Frac(None=@n46)) {constB=4}, B=Frac(None=Multiply(A=@n47) {constB=-0.02}))
  n49 = Constant() {r=3}
  n50 = Multiply(A=Sine(None=@n48), B=@n49) {desc=Amplitude}
  n51 = Multiply(A=Cosine(None=@n48), B=@n49)
  n52 = OneMinus(None=ComponentMask(None=TextureCoordinate()) {g=true})
  n53 = Multiply(
   A=MaterialFunctionCall(X=ComponentMask(None=@n50) {r=true}, Y=ComponentMask(None=@n51) {r=true}, Z=Constant()) {materialFunction=&MakeFloat3}.Result
   B=Multiply(A=@n52, B=@n52)
  )
  n54 = Clamp(None=@n2)
  n55 = OneMinus(None=ComponentMask(None=@n1) {g=true})
  n56 = Sine(None=Time()) {period=0.5}
  n57 = Sine(None=Time()) {period=0.73}
  n58 = Multiply(A=Add(A=Multiply(A=@n56) {constB=0.79}, B=Multiply(A=@n57) {constB=0.5})) {constB=5}
  n59 = ComponentMask(None=OneMinus(None=TextureCoordinate())) {g=true}
  n60 = Add(A=Multiply(A=Add(A=@n45, B=@n53) {constB=0}, B=Multiply(A=@n54, B=@n55)), B=AppendVector(A=Constant2Vector(), B=Multiply(A=@n58, B=@n59)))
  n61 = TextureCoordinate()
  n62 = Multiply(A=ComponentMask(None=@n61) {g=true}) {constB=4}
  n63 = Multiply(A=OneMinus(None=ComponentMask(None=@n61) {g=true}))
  n64 = Multiply(A=Multiply(A=VertexNormalWS()) {constB=50}, B=Multiply(A=Clamp(None=@n62), B=Clamp(None=@n63)))
  NamedRerouteDeclaration(None=Add(
    A=Add(A=@n60, B=@n64)
    B=MaterialFunctionCall() {materialFunction=/Game/Athena/Aircraft/PartyBus_RT/Materials/MF_PartyBusAnimationWPO}.Result
   )) {name=WPO, nodeColor=<rgba 1 0.5156 0 1>}
  NamedRerouteUsage()
  OneMinus(None=@n18)
  Power(Base=@n4.Fresnel) {constExponent=20}
  n65 = Divide(A=@n15, B=ScalarParameter() {defaultValue=0.0666, parameterName=Top Fade Softness})
  Power(Base=Multiply(A=Saturate(None=@n65), B=@n4.Fresnel), Exp=ScalarParameter() {defaultValue=3, parameterName=Fresnel Exponent})
  Saturate(None=Power(Base=@n9))
  n66 = Clamp(None=Multiply(A=NamedRerouteUsage(), B=NamedRerouteUsage()))
  StaticSwitchParameter(True=DepthFade(Opacity=@n66, FadeDistance=ScalarParameter() {defaultValue=30, parameterName=DepthFade}), False=@n66) {parameterName=DepthFade}
  n67 = Multiply(A=NamedRerouteUsage(), B=ScalarParameter() {defaultValue=1, parameterName=Emission}) {constB=1.2}
  StaticSwitchParameter(
   True=MaterialFunctionCall(Hue Shift Percentage=ScalarParameter() {parameterName=HueShift}, Texture=@n67) {materialFunction=&HueShift}.Result
   False=@n67
  ) {parameterName=DepthFade}
 ]
}
