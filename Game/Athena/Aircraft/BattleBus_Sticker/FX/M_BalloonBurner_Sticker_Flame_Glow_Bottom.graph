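# Additive unlit material for the bottom flame glow of the Sticker Battle Bus balloon burner.
# A radial sphere mask blends between two colors and shapes opacity along with a scene-versus-pixel depth difference.
# Two multiplied time-driven sines form a Flickering signal to modulate the glow, and a DepthFade softens intersections.
# It also calls the shared Party Bus animation WPO function.
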Material /Game/Athena/Aircraft/BattleBus_Sticker/FX/M_BalloonBurner_Sticker_Flame_Glow_Bottom
graph {
 properties {blendMode=BLEND_Additive, pixelDepthOffsetMode=PDOM_Legacy, shadingModel=MSM_Unlit, usedShadingModels=Unlit}
 nodes [
  n1 = SceneDepth()
  n2 = PixelDepth()
  n3 = OneMinus(None=Multiply(A=Subtract(A=@n1.R, B=@n2.R), B=ScalarParameter() {defaultValue=0.01, parameterName=Depth}))
  n4 = Power(Base=SphereMask(A=TextureCoordinate(), B=Constant2Vector() {g=0.5, r=0.5}) {attenuationRadius=0.5, hardnessPercent=1})
  n5 = NamedRerouteUsage()
  DepthFade(
   Opacity=Multiply(A=Multiply(A=@n3, B=@n4), B=LinearInterpolate(Alpha=@n5) {constA=0.8})
   FadeDistance=ScalarParameter() {defaultValue=3000, parameterName=Fade Distance}
  ) {fadeDistanceDefault=32}
  n6 = VectorParameter() {parameterName=Color B, defaultValue=<rgba 0.3958 1.8886 2 0.1>}
  n7 = TextureCoordinate()
  n8 = Constant2Vector() {g=0.5, r=0.5}
  LinearInterpolate(
   A=VectorParameter() {parameterName=Color A, defaultValue=<rgba 1 0.8421 0 0.2>}.RGB
   B=@n6.RGB
   Alpha=Power(Base=SphereMask(A=@n7, B=@n8) {attenuationRadius=0.5, hardnessPercent=1}) {constExponent=1}
  )
  MaterialFunctionCall() {materialFunction=/Game/Athena/Aircraft/PartyBus_RT/Materials/MF_PartyBusAnimationWPO}
  Multiply(A=Multiply(A=@n6.RGB, B=ScalarParameter() {defaultValue=20, parameterName=Intensity}) {constB=0})
  n9 = Time()
  n10 = Sine(None=Multiply(A=@n9) {constB=1.3})
  n11 = Sine(None=Multiply(A=@n9) {constB=2.8})
  NamedRerouteDeclaration(None=ConstantBiasScale(None=Multiply(A=@n10, B=@n11))) {name=Flickering, nodeColor=<rgba 0.2891 0 1 1>}
 ]
}
